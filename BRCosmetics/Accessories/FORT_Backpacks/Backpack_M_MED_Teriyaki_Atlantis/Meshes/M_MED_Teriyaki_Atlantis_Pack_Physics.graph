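# Physics asset for the Teriyaki Atlantis backpack cosmetic, giving physical bodies and joints to the pack and its left and right dyn_sword bones.
# Each sword bone carries a mirrored capsule, while the pack bone has a small, near-spherical capsule.
# Two matching constraints join the swords to the pack with narrow swing and even tighter twist limits, restricting how far each can move relative to it.
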PhysicsAsset /BRCosmetics/Accessories/FORT_Backpacks/Backpack_M_MED_Teriyaki_Atlantis/Meshes/M_MED_Teriyaki_Atlantis_Pack_Physics
bodies {
 dyn_sword_l [{shapeType=Capsule, length=25.008, radius=8.5382, center=<2.6909 -0.0022 -0.1171>, rotation=<rot 0 -90.0463 -87.7057>}]
 dyn_sword_r [{shapeType=Capsule, length=25.008, radius=8.5382, center=<2.6919 -0.0017 0.0936>, rotation=<rot 0 89.9635 -87.8631>}]
 pack [{shapeType=Capsule, length=0.0005, radius=2.6438}]
}
constraints [
 {bone1Name=dyn_sword_l, bone2Name=pack, swing1LimitDegrees=15, swing2LimitDegrees=15, twistLimitDegrees=5}
 {bone1Name=dyn_sword_r, bone2Name=pack, swing1LimitDegrees=15, swing2LimitDegrees=15, twistLimitDegrees=5}
]
